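# Physics asset for the CoyoteTrail pickaxe mesh, covering a two-segment trail chain on the weapon.
# Two small sphere bodies sit on the Dyn_Trail_01 and Dyn_Trail_02 bones.
# A single constraint links them with moderate swing limits and a tighter twist limit.
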PhysicsAsset /BRCosmetics/Weapons/FORT_Melee/Pickaxe_CoyoteTrail/Meshes/CoyoteTrail_Axe_Physics
bodies {
 Dyn_Trail_01 [{shapeType=Sphere, radius=3.5, shapeName=Dyn_Trail_01_sphere}]
 Dyn_Trail_02 [{shapeType=Sphere, radius=3.5, shapeName=Dyn_Trail_02_sphere}]
}
constraints [{bone1Name=Dyn_Trail_01, bone2Name=Dyn_Trail_02, swing1LimitDegrees=30, swing2LimitDegrees=30, twistLimitDegrees=10}]
